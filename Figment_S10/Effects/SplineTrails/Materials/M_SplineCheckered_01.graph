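# M_SplineCheckered_01 is a Default Lit material for spline meshes in the SplineTrails effects.
# It tiles the U texture coordinate six times, takes its fractional part and rounds it, producing a hard-edged alternating 0/1 band mask.
# Despite the checkered name, the pattern varies along U only, and the graph shows no connection from it to a material output.
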Material /Figment_S10/Effects/SplineTrails/Materials/M_SplineCheckered_01
graph {
 properties {bUsedWithSplineMeshes=true, pixelDepthOffsetMode=PDOM_Legacy, usedShadingModels=Default Lit}
 nodes [
  n1 = TextureCoordinate() {uTiling=6}
  MaterialFunctionCall!(In=Frac(None=ComponentMask(None=@n1) {r=true})) {materialFunction=/Engine/Functions/Engine_MaterialFunctions02/Utility/Round}
 ]
}
